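# Material layer blend function that combines Base and Top Material attributes into one Blended Material using a scalar Alpha.
# Most connected attributes are linearly interpolated by Alpha.
# Normals instead go through BlendAngleCorrectedNormals, which ignores Alpha, so Top normal detail applies regardless of blend weight.
# World position offset comes from Base only, while Refraction and all eight customized UVs are wired to the Top Material's Refraction output.
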MaterialFunction /FigmentEnvironment/LM_8_1_1/Environments/Sets/Arid/Materials/MatLayerBlend_StandardCorrectNormals
graph {
 properties {description=Blends all attributes of 2 Materials, libraryCategoriesText [Misc, MaterialLayerBlend]}
 nodes [
  n1 = BreakMaterialAttributes!(Attr=FunctionInput() {bUsePreviewValueAsDefault=true, inputName=Base Material, inputType=FunctionInput_MaterialAttributes})
  n2 = BreakMaterialAttributes(Attr=FunctionInput() {bUsePreviewValueAsDefault=true, inputName=Top Material, inputType=FunctionInput_MaterialAttributes, sortPriority=1})
  n3 = FunctionInput() {bUsePreviewValueAsDefault=true, inputName=Alpha, inputType=FunctionInput_Scalar, sortPriority=2}
  FunctionOutput(None=MakeMaterialAttributes!(
    BaseColor=LinearInterpolate(A=@n1.BaseColor, B=@n2.BaseColor, Alpha=@n3)
    Metallic=LinearInterpolate(A=@n1.Metallic, B=@n2.Metallic, Alpha=@n3)
    Specular=LinearInterpolate(A=@n1.Specular, B=@n2.Specular, Alpha=@n3)
    Roughness=LinearInterpolate(A=@n1.Roughness, B=@n2.Roughness, Alpha=@n3)
    Anisotropy
    EmissiveColor=LinearInterpolate(A=@n1.EmissiveColor, B=@n2.EmissiveColor, Alpha=@n3)
    Opacity=LinearInterpolate(A=@n1.Opacity, B=@n2.Opacity, Alpha=@n3)
    OpacityMask=LinearInterpolate(A=@n1.OpacityMask, B=@n2.OpacityMask, Alpha=@n3)
    Normal=MaterialFunctionCall(BaseNormal=@n1.Normal, AdditionalNormal=@n2.Normal) {materialFunction=&BlendAngleCorrectedNormals}.Result
    Tangent
    WorldPositionOffset=@n1.WorldPositionOffset
    SubsurfaceColor=LinearInterpolate(A=@n1.SubsurfaceColor, B=@n2.SubsurfaceColor, Alpha=@n3)
    ClearCoat=LinearInterpolate(A=@n1.ClearCoat, B=@n2.ClearCoat, Alpha=@n3)
    ClearCoatRoughness=LinearInterpolate(A=@n1.ClearCoatRoughness, B=@n2.ClearCoatRoughness, Alpha=@n3)
    AmbientOcclusion=LinearInterpolate(A=@n1.AmbientOcclusion, B=@n2.AmbientOcclusion, Alpha=@n3)
    Refraction=@n2.Refraction
    CustomizedUVs_0=@n2.Refraction
    CustomizedUVs_1=@n2.Refraction
    CustomizedUVs_2=@n2.Refraction
    CustomizedUVs_3=@n2.Refraction
    CustomizedUVs_4=@n2.Refraction
    CustomizedUVs_5=@n2.Refraction
    CustomizedUVs_6=@n2.Refraction
    CustomizedUVs_7=@n2.Refraction
    PixelDepthOffset=LinearInterpolate(A=@n1.PixelDepthOffset, B=@n2.PixelDepthOffset, Alpha=@n3)
    ShadingModel
    Displacement
   )) {outputName=Blended Material, sortPriority=-10}
  LinearInterpolate(A=@n1.Normal, B=@n2.Normal, Alpha=@n3)
  LinearInterpolate(A=@n1.WorldPositionOffset, B=@n2.WorldPositionOffset, Alpha=@n3)
  LinearInterpolate(Alpha=@n3)
  LinearInterpolate(Alpha=@n3)
 ]
}
